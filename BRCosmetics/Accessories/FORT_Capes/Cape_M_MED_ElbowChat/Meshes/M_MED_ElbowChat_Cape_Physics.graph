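# Physics asset for the ElbowChat medium male cape cosmetic, defining capsule bodies and joints for cape motion.
# Five cape chains each run from a master bone attached to spine_05 through seven segments, with tight twist and swing limits varying along the length.
# Cross constraints with free rotation link neighboring chains row by row, forming a cloth-like lattice.
# Pelvis, spine, thigh, calf and foot capsules supply the wearer's body shape.
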PhysicsAsset /BRCosmetics/Accessories/FORT_Capes/Cape_M_MED_ElbowChat/Meshes/M_MED_ElbowChat_Cape_Physics
bodies {
 calf_l [{shapeType=Capsule, length=43.2612, radius=7.2059, shapeName=calf_l_capsule, center=<-18.4728 0 0>, rotation=<rot 0 -90.0002 -89.0579>}]
 calf_r [{shapeType=Capsule, length=43.2612, radius=7.2059, shapeName=calf_l_capsule, center=<18.4728 0 0>, rotation=<rot 0 -90.0002 90.9421>}]
 dyn_cape_01 [{shapeType=Capsule, length=10.8075, radius=3.1925, shapeName=dyn_cape_01_capsule, center=<7 0 0>, rotation=<rot 0 90 90>}]
 dyn_cape_02 [{shapeType=Capsule, length=8, radius=6, shapeName=dyn_cape_02_capsule, center=<7 0 0>, rotation=<rot 0 90 90>}]
 dyn_cape_03 [{shapeType=Capsule, length=8, radius=6, shapeName=dyn_cape_03_capsule, center=<7 0 0>, rotation=<rot 0 90 90>}]
 dyn_cape_04 [{shapeType=Capsule, length=8, radius=6, shapeName=dyn_cape_04_capsule, center=<7 0 0>, rotation=<rot 0 90 90>}]
 dyn_cape_05 [{shapeType=Capsule, length=8, radius=6, shapeName=dyn_cape_05_capsule, center=<7 0 0>, rotation=<rot 0 90 90>}]
 dyn_cape_06 [{shapeType=Capsule, length=8, radius=6, shapeName=dyn_cape_06_capsule, center=<7 0 0>, rotation=<rot 0 90 90>}]
 dyn_cape_07 [{shapeType=Capsule, length=10.1193, radius=9.1676, shapeName=dyn_cape_07_capsule, center=<7 0 0>, rotation=<rot 0 90 90>}]
 dyn_cape_MSTR [{shapeType=Capsule, length=0.505, radius=0.505, shapeName=dyn_cape_MSTR_capsule, rotation=<rot 0 -90 0>}]
 dyn_cape_l_01 [{shapeType=Capsule, length=10.8075, radius=3.1925, center=<7 0 0>, rotation=<rot 0 90 90>}]
 dyn_cape_l_02 [{shapeType=Capsule, length=8, radius=6, shapeName=dyn_cape_l_02_capsule, center=<7 0 0>, rotation=<rot 0 90 90>}]
 dyn_cape_l_03 [{shapeType=Capsule, length=8, radius=6, shapeName=dyn_cape_l_03_capsule, center=<7 0 0>, rotation=<rot 0 90 90>}]
 dyn_cape_l_04 [{shapeType=Capsule, length=8, radius=6, center=<7 0 0>, rotation=<rot 0 90 90>}]
 dyn_cape_l_05 [{shapeType=Capsule, length=8, radius=6, shapeName=dyn_cape_l_05_capsule, center=<7 0 0>, rotation=<rot 0 90 90>}]
 dyn_cape_l_06 [{shapeType=Capsule, length=8, radius=6, center=<7 0 0>, rotation=<rot 0 90 90>}]
 dyn_cape_l_07 [{shapeType=Capsule, length=10.1193, radius=9.1676, shapeName=dyn_cape_l_07_capsule, center=<7 0 0>, rotation=<rot 0 90 90>}]
 dyn_cape_l_MSTR [{shapeType=Capsule, length=0.505, radius=0.505, shapeName=dyn_cape_l_MSTR_capsule, rotation=<rot 0 -90 0>}]
 dyn_cape_mid_l_01 [{shapeType=Capsule, length=10.8075, radius=3.1925, shapeName=dyn_cape_mid_l_01_capsule, center=<7 0 0>, rotation=<rot 0 90 90>}]
 dyn_cape_mid_l_02 [{shapeType=Capsule, length=8, radius=6, center=<7 0 0>, rotation=<rot 0 90 90>}]
 dyn_cape_mid_l_03 [{shapeType=Capsule, length=8, radius=6, shapeName=dyn_cape_mid_l_03_capsule, center=<7 0 0>, rotation=<rot 0 90 90>}]
 dyn_cape_mid_l_04 [{shapeType=Capsule, length=8, radius=6, shapeName=dyn_cape_mid_l_04_capsule, center=<7 0 0>, rotation=<rot 0 90 90>}]
 dyn_cape_mid_l_05 [{shapeType=Capsule, length=8, radius=6, center=<7 0 0>, rotation=<rot 0 90 90>}]
 dyn_cape_mid_l_06 [{shapeType=Capsule, length=8, radius=6, shapeName=dyn_cape_mid_l_06_capsule, center=<7 0 0>, rotation=<rot 0 90 90>}]
 dyn_cape_mid_l_07 [{shapeType=Capsule, length=10.1193, radius=9.1676, shapeName=dyn_cape_mid_l_07_capsule, center=<7 0 0>, rotation=<rot 0 90 90>}]
 dyn_cape_mid_l_MSTR [{shapeType=Capsule, length=0.505, radius=0.505, shapeName=dyn_cape_mid_l_MSTR_capsule, rotation=<rot 0 -90 0>}]
 dyn_cape_mid_r_01 [{shapeType=Capsule, length=10.8075, radius=3.1925, shapeName=dyn_cape_mid_r_01_capsule, center=<-7 0 0>, rotation=<rot 0 90 90>}]
 dyn_cape_mid_r_02 [{shapeType=Capsule, length=8, radius=6, center=<-7 0 0>, rotation=<rot 0 90 90>}]
 dyn_cape_mid_r_03 [{shapeType=Capsule, length=8, radius=6, shapeName=dyn_cape_mid_r_03_capsule, center=<-7 0 0>, rotation=<rot 0 90 90>}]
 dyn_cape_mid_r_04 [{shapeType=Capsule, length=8, radius=6, shapeName=dyn_cape_mid_r_04_capsule, center=<-7 0 0>, rotation=<rot 0 90 90>}]
 dyn_cape_mid_r_05 [{shapeType=Capsule, length=8, radius=6, center=<-7 0 0>, rotation=<rot 0 90 90>}]
 dyn_cape_mid_r_06 [{shapeType=Capsule, length=8, radius=6, shapeName=dyn_cape_mid_r_06_capsule, center=<-7 0 0>, rotation=<rot 0 90 90>}]
 dyn_cape_mid_r_07 [{shapeType=Capsule, length=10.1193, radius=9.1676, shapeName=dyn_cape_mid_r_07_capsule, center=<-7 0 0>, rotation=<rot 0 90 90>}]
 dyn_cape_mid_r_MSTR [{shapeType=Capsule, length=0.505, radius=0.505, shapeName=dyn_cape_mid_r_MSTR_capsule, rotation=<rot 0 -90 0>}]
 dyn_cape_r_01 [{shapeType=Capsule, length=10.8075, radius=3.1925, shapeName=dyn_cape_r_01_capsule, center=<-7 0 0>, rotation=<rot 0 90 90>}]
 dyn_cape_r_02 [{shapeType=Capsule, length=8, radius=6, shapeName=dyn_cape_r_02_capsule, center=<-7 0 0>, rotation=<rot 0 90 90>}]
 dyn_cape_r_03 [{shapeType=Capsule, length=8, radius=6, center=<-7 0 0>, rotation=<rot 0 90 90>}]
 dyn_cape_r_04 [{shapeType=Capsule, length=8, radius=6, shapeName=dyn_cape_r_04_capsule, center=<-7 0 0>, rotation=<rot 0 90 90>}]
 dyn_cape_r_05 [{shapeType=Capsule, length=8, radius=6, shapeName=dyn_cape_r_05_capsule, center=<-7 0 0>, rotation=<rot 0 90 90>}]
 dyn_cape_r_06 [{shapeType=Capsule, length=8, radius=6, center=<-7 0 0>, rotation=<rot 0 90 90>}]
 dyn_cape_r_07 [{shapeType=Capsule, length=10.1193, radius=9.1676, shapeName=dyn_cape_r_07_capsule, center=<-7 0 0>, rotation=<rot 0 90 90>}]
 dyn_cape_r_MSTR [{shapeType=Capsule, length=0.505, radius=0.505, shapeName=dyn_cape_r_MSTR_capsule, rotation=<rot 0 -90 0>}]
 foot_l [{
   shapeType=Capsule
   length=15.181
   radius=9.748
   shapeName=foot_l_capsule
   center=<-3.0207 5.4437 -0.8247>
   rotation=<rot 81.3855 89.9999 180>
  }]
 foot_r [{shapeType=Capsule, length=15.181, radius=9.748, shapeName=foot_l_capsule, center=<3.0207 -5.4437 0.8247>, rotation=<rot 81.3855 89.9999 0>}]
 pelvis [{shapeType=Capsule, length=11.1879, radius=12.2556, shapeName=pelvis_capsule, center=<0 -1.8018 0>, rotation=<rot 0 -90 0>}]
 spine_01 [{shapeType=Capsule, length=2.4425, radius=12.0941, shapeName=spine_01_capsule, rotation=<rot 0 -90 0>}]
 spine_02 [{shapeType=Capsule, length=6.0993, radius=10.2738, shapeName=spine_02_capsule, rotation=<rot 0 -90 0>}]
 spine_03 [{shapeType=Capsule, length=3.9484, radius=11.6846, shapeName=spine_03_capsule, center=<1.8017 1.3357 0>, rotation=<rot 0 -90 0>}]
 spine_04 [{shapeType=Capsule, length=4.412, radius=12.6834, shapeName=spine_04_capsule, center=<1.198 1.7431 0>, rotation=<rot 0 -90 0>}]
 spine_05 [{shapeType=Capsule, length=13.1379, radius=11.3508, center=<-3.078 2.0801 0.1911>, rotation=<rot -1.6624 -90 0.7537>}]
 thigh_l [{
   shapeType=Capsule
   length=34.0785
   radius=8.3327
   shapeName=thigh_l_capsule
   center=<-16.5161 0.0001 -0.0308>
   rotation=<rot 0 -90.0002 -90.1068>
  }]
 thigh_r [{
   shapeType=Capsule
   length=34.0785
   radius=8.3327
   shapeName=thigh_l_capsule
   center=<16.5161 -0.0001 0.0308>
   rotation=<rot 0 -90.0002 89.8932>
  }]
}
constraints [
 {bone1Name=dyn_cape_mid_l_02, bone2Name=dyn_cape_mid_l_01, swing1LimitDegrees=20, swing2LimitDegrees=45, twistLimitDegrees=4}
 {bone1Name=dyn_cape_mid_l_05, bone2Name=dyn_cape_mid_l_04, swing1LimitDegrees=45, swing2LimitDegrees=45, twistLimitDegrees=4}
 {bone1Name=dyn_cape_l_01, bone2Name=dyn_cape_l_MSTR, swing1LimitDegrees=20, swing2LimitDegrees=45, twistLimitDegrees=4}
 {bone1Name=dyn_cape_l_04, bone2Name=dyn_cape_l_03, swing1LimitDegrees=45, swing2LimitDegrees=45, twistLimitDegrees=4}
 {bone1Name=dyn_cape_l_06, bone2Name=dyn_cape_l_05, swing1LimitDegrees=18, swing2LimitDegrees=18, twistLimitDegrees=4}
 {bone1Name=dyn_cape_r_03, bone2Name=dyn_cape_r_02, swing1LimitDegrees=45, swing2LimitDegrees=45, twistLimitDegrees=4}
 {bone1Name=dyn_cape_r_06, bone2Name=dyn_cape_r_05, swing1LimitDegrees=18, swing2LimitDegrees=18, twistLimitDegrees=4}
 {bone1Name=dyn_cape_mid_r_02, bone2Name=dyn_cape_mid_r_01, swing1LimitDegrees=20, swing2LimitDegrees=45, twistLimitDegrees=4}
 {bone1Name=dyn_cape_mid_r_05, bone2Name=dyn_cape_mid_r_04, swing1LimitDegrees=45, swing2LimitDegrees=45, twistLimitDegrees=4}
 {
  bone1Name=spine_05
  bone2Name=spine_04
  swing1LimitDegrees=45
  swing1Motion=Free
  swing2LimitDegrees=45
  swing2Motion=Free
  twistLimitDegrees=45
  twistMotion=Free
 }
 {
  bone1Name=spine_02
  bone2Name=spine_01
  swing1LimitDegrees=45
  swing1Motion=Free
  swing2LimitDegrees=45
  swing2Motion=Free
  twistLimitDegrees=45
  twistMotion=Free
 }
 {
  bone1Name=spine_03
  bone2Name=spine_02
  swing1LimitDegrees=45
  swing1Motion=Free
  swing2LimitDegrees=45
  swing2Motion=Free
  twistLimitDegrees=45
  twistMotion=Free
 }
 {
  bone1Name=spine_04
  bone2Name=spine_03
  swing1LimitDegrees=45
  swing1Motion=Free
  swing2LimitDegrees=45
  swing2Motion=Free
  twistLimitDegrees=45
  twistMotion=Free
 }
 {
  bone1Name=spine_01
  bone2Name=pelvis
  swing1LimitDegrees=45
  swing1Motion=Free
  swing2LimitDegrees=45
  swing2Motion=Free
  twistLimitDegrees=45
  twistMotion=Free
 }
 {
  bone1Name=thigh_l
  bone2Name=pelvis
  swing1LimitDegrees=45
  swing1Motion=Free
  swing2LimitDegrees=45
  swing2Motion=Free
  twistLimitDegrees=45
  twistMotion=Free
 }
 {
  bone1Name=calf_l
  bone2Name=thigh_l
  swing1LimitDegrees=45
  swing1Motion=Free
  swing2LimitDegrees=45
  swing2Motion=Free
  twistLimitDegrees=45
  twistMotion=Free
 }
 {
  bone1Name=foot_l
  bone2Name=calf_l
  swing1LimitDegrees=45
  swing1Motion=Free
  swing2LimitDegrees=45
  swing2Motion=Free
  twistLimitDegrees=45
  twistMotion=Free
 }
 {
  bone1Name=dyn_cape_MSTR
  bone2Name=spine_05
  swing1LimitDegrees=45
  swing1Motion=Free
  swing2LimitDegrees=45
  swing2Motion=Free
  twistLimitDegrees=45
  twistMotion=Free
 }
 {bone1Name=dyn_cape_01, bone2Name=dyn_cape_MSTR, swing1LimitDegrees=20, swing2LimitDegrees=45, twistLimitDegrees=4}
 {bone1Name=dyn_cape_02, bone2Name=dyn_cape_01, swing1LimitDegrees=20, swing2LimitDegrees=45, twistLimitDegrees=4}
 {bone1Name=dyn_cape_03, bone2Name=dyn_cape_02, swing1LimitDegrees=45, swing2LimitDegrees=45, twistLimitDegrees=4}
 {bone1Name=dyn_cape_04, bone2Name=dyn_cape_03, swing1LimitDegrees=45, swing2LimitDegrees=45, twistLimitDegrees=4}
 {bone1Name=dyn_cape_05, bone2Name=dyn_cape_04, swing1LimitDegrees=45, swing2LimitDegrees=45, twistLimitDegrees=4}
 {bone1Name=dyn_cape_06, bone2Name=dyn_cape_05, swing1LimitDegrees=18, swing2LimitDegrees=18, twistLimitDegrees=4}
 {bone1Name=dyn_cape_07, bone2Name=dyn_cape_06, swing1LimitDegrees=18, swing2LimitDegrees=18, twistLimitDegrees=4}
 {bone1Name=dyn_cape_l_02, bone2Name=dyn_cape_l_01, swing1LimitDegrees=20, swing2LimitDegrees=45, twistLimitDegrees=4}
 {bone1Name=dyn_cape_l_03, bone2Name=dyn_cape_l_02, swing1LimitDegrees=45, swing2LimitDegrees=45, twistLimitDegrees=4}
 {bone1Name=dyn_cape_l_05, bone2Name=dyn_cape_l_04, swing1LimitDegrees=45, swing2LimitDegrees=45, twistLimitDegrees=4}
 {bone1Name=dyn_cape_l_07, bone2Name=dyn_cape_l_06, swing1LimitDegrees=18, swing2LimitDegrees=18, twistLimitDegrees=4}
 {
  bone1Name=dyn_cape_mid_l_MSTR
  bone2Name=spine_05
  swing1LimitDegrees=45
  swing1Motion=Free
  swing2LimitDegrees=45
  swing2Motion=Free
  twistLimitDegrees=45
  twistMotion=Free
 }
 {bone1Name=dyn_cape_mid_l_01, bone2Name=dyn_cape_mid_l_MSTR, swing1LimitDegrees=20, swing2LimitDegrees=45, twistLimitDegrees=4}
 {bone1Name=dyn_cape_mid_l_03, bone2Name=dyn_cape_mid_l_02, swing1LimitDegrees=45, swing2LimitDegrees=45, twistLimitDegrees=4}
 {bone1Name=dyn_cape_mid_l_04, bone2Name=dyn_cape_mid_l_03, swing1LimitDegrees=45, swing2LimitDegrees=45, twistLimitDegrees=4}
 {bone1Name=dyn_cape_mid_l_06, bone2Name=dyn_cape_mid_l_05, swing1LimitDegrees=18, swing2LimitDegrees=18, twistLimitDegrees=4}
 {bone1Name=dyn_cape_mid_l_07, bone2Name=dyn_cape_mid_l_06, swing1LimitDegrees=18, swing2LimitDegrees=18, twistLimitDegrees=4}
 {
  bone1Name=dyn_cape_mid_r_MSTR
  bone2Name=spine_05
  swing1LimitDegrees=45
  swing1Motion=Free
  swing2LimitDegrees=45
  swing2Motion=Free
  twistLimitDegrees=45
  twistMotion=Free
 }
 {bone1Name=dyn_cape_mid_r_01, bone2Name=dyn_cape_mid_r_MSTR, swing1LimitDegrees=20, swing2LimitDegrees=45, twistLimitDegrees=4}
 {bone1Name=dyn_cape_mid_r_03, bone2Name=dyn_cape_mid_r_02, swing1LimitDegrees=45, swing2LimitDegrees=45, twistLimitDegrees=4}
 {bone1Name=dyn_cape_mid_r_04, bone2Name=dyn_cape_mid_r_03, swing1LimitDegrees=45, swing2LimitDegrees=45, twistLimitDegrees=4}
 {bone1Name=dyn_cape_mid_r_06, bone2Name=dyn_cape_mid_r_05, swing1LimitDegrees=18, swing2LimitDegrees=18, twistLimitDegrees=4}
 {bone1Name=dyn_cape_mid_r_07, bone2Name=dyn_cape_mid_r_06, swing1LimitDegrees=18, swing2LimitDegrees=18, twistLimitDegrees=4}
 {
  bone1Name=dyn_cape_r_MSTR
  bone2Name=spine_05
  swing1LimitDegrees=45
  swing1Motion=Free
  swing2LimitDegrees=45
  swing2Motion=Free
  twistLimitDegrees=45
  twistMotion=Free
 }
 {bone1Name=dyn_cape_r_01, bone2Name=dyn_cape_r_MSTR, swing1LimitDegrees=20, swing2LimitDegrees=45, twistLimitDegrees=4}
 {bone1Name=dyn_cape_r_02, bone2Name=dyn_cape_r_01, swing1LimitDegrees=20, swing2LimitDegrees=45, twistLimitDegrees=4}
 {bone1Name=dyn_cape_r_04, bone2Name=dyn_cape_r_03, swing1LimitDegrees=45, swing2LimitDegrees=45, twistLimitDegrees=4}
 {bone1Name=dyn_cape_r_05, bone2Name=dyn_cape_r_04, swing1LimitDegrees=45, swing2LimitDegrees=45, twistLimitDegrees=4}
 {bone1Name=dyn_cape_r_07, bone2Name=dyn_cape_r_06, swing1LimitDegrees=18, swing2LimitDegrees=18, twistLimitDegrees=4}
 {
  bone1Name=dyn_cape_l_MSTR
  bone2Name=spine_05
  swing1LimitDegrees=45
  swing1Motion=Free
  swing2LimitDegrees=45
  swing2Motion=Free
  twistLimitDegrees=45
  twistMotion=Free
 }
 {
  bone1Name=thigh_r
  bone2Name=pelvis
  swing1LimitDegrees=45
  swing1Motion=Free
  swing2LimitDegrees=45
  swing2Motion=Free
  twistLimitDegrees=45
  twistMotion=Free
 }
 {
  bone1Name=calf_r
  bone2Name=thigh_r
  swing1LimitDegrees=45
  swing1Motion=Free
  swing2LimitDegrees=45
  swing2Motion=Free
  twistLimitDegrees=45
  twistMotion=Free
 }
 {
  bone1Name=foot_r
  bone2Name=calf_r
  swing1LimitDegrees=45
  swing1Motion=Free
  swing2LimitDegrees=45
  swing2Motion=Free
  twistLimitDegrees=45
  twistMotion=Free
 }
 {
  bone1Name=dyn_cape_01
  bone2Name=dyn_cape_mid_l_01
  swing1LimitDegrees=45
  swing1Motion=Free
  swing2LimitDegrees=45
  swing2Motion=Free
  twistLimitDegrees=45
  twistMotion=Free
 }
 {
  bone1Name=dyn_cape_01
  bone2Name=dyn_cape_mid_r_01
  swing1LimitDegrees=45
  swing1Motion=Free
  swing2LimitDegrees=45
  swing2Motion=Free
  twistLimitDegrees=45
  twistMotion=Free
 }
 {
  bone1Name=dyn_cape_02
  bone2Name=dyn_cape_mid_l_02
  swing1LimitDegrees=45
  swing1Motion=Free
  swing2LimitDegrees=45
  swing2Motion=Free
  twistLimitDegrees=45
  twistMotion=Free
 }
 {
  bone1Name=dyn_cape_02
  bone2Name=dyn_cape_mid_r_02
  swing1LimitDegrees=45
  swing1Motion=Free
  swing2LimitDegrees=45
  swing2Motion=Free
  twistLimitDegrees=45
  twistMotion=Free
 }
 {
  bone1Name=dyn_cape_03
  bone2Name=dyn_cape_mid_l_03
  swing1LimitDegrees=45
  swing1Motion=Free
  swing2LimitDegrees=45
  swing2Motion=Free
  twistLimitDegrees=45
  twistMotion=Free
 }
 {
  bone1Name=dyn_cape_03
  bone2Name=dyn_cape_mid_r_03
  swing1LimitDegrees=45
  swing1Motion=Free
  swing2LimitDegrees=45
  swing2Motion=Free
  twistLimitDegrees=45
  twistMotion=Free
 }
 {
  bone1Name=dyn_cape_04
  bone2Name=dyn_cape_mid_l_04
  swing1LimitDegrees=45
  swing1Motion=Free
  swing2LimitDegrees=45
  swing2Motion=Free
  twistLimitDegrees=45
  twistMotion=Free
 }
 {
  bone1Name=dyn_cape_04
  bone2Name=dyn_cape_mid_r_04
  swing1LimitDegrees=45
  swing1Motion=Free
  swing2LimitDegrees=45
  swing2Motion=Free
  twistLimitDegrees=45
  twistMotion=Free
 }
 {
  bone1Name=dyn_cape_05
  bone2Name=dyn_cape_mid_l_05
  swing1LimitDegrees=45
  swing1Motion=Free
  swing2LimitDegrees=45
  swing2Motion=Free
  twistLimitDegrees=45
  twistMotion=Free
 }
 {
  bone1Name=dyn_cape_05
  bone2Name=dyn_cape_mid_r_05
  swing1LimitDegrees=45
  swing1Motion=Free
  swing2LimitDegrees=45
  swing2Motion=Free
  twistLimitDegrees=45
  twistMotion=Free
 }
 {
  bone1Name=dyn_cape_06
  bone2Name=dyn_cape_mid_l_06
  swing1LimitDegrees=45
  swing1Motion=Free
  swing2LimitDegrees=45
  swing2Motion=Free
  twistLimitDegrees=45
  twistMotion=Free
 }
 {
  bone1Name=dyn_cape_06
  bone2Name=dyn_cape_mid_r_06
  swing1LimitDegrees=45
  swing1Motion=Free
  swing2LimitDegrees=45
  swing2Motion=Free
  twistLimitDegrees=45
  twistMotion=Free
 }
 {
  bone1Name=dyn_cape_07
  bone2Name=dyn_cape_mid_l_07
  swing1LimitDegrees=45
  swing1Motion=Free
  swing2LimitDegrees=45
  swing2Motion=Free
  twistLimitDegrees=45
  twistMotion=Free
 }
 {
  bone1Name=dyn_cape_07
  bone2Name=dyn_cape_mid_r_07
  swing1LimitDegrees=45
  swing1Motion=Free
  swing2LimitDegrees=45
  swing2Motion=Free
  twistLimitDegrees=45
  twistMotion=Free
 }
 {
  bone1Name=dyn_cape_mid_l_01
  bone2Name=dyn_cape_l_01
  swing1LimitDegrees=45
  swing1Motion=Free
  swing2LimitDegrees=45
  swing2Motion=Free
  twistLimitDegrees=45
  twistMotion=Free
 }
 {
  bone1Name=dyn_cape_mid_l_02
  bone2Name=dyn_cape_l_02
  swing1LimitDegrees=45
  swing1Motion=Free
  swing2LimitDegrees=45
  swing2Motion=Free
  twistLimitDegrees=45
  twistMotion=Free
 }
 {
  bone1Name=dyn_cape_mid_l_03
  bone2Name=dyn_cape_l_03
  swing1LimitDegrees=45
  swing1Motion=Free
  swing2LimitDegrees=45
  swing2Motion=Free
  twistLimitDegrees=45
  twistMotion=Free
 }
 {
  bone1Name=dyn_cape_mid_l_04
  bone2Name=dyn_cape_l_04
  swing1LimitDegrees=45
  swing1Motion=Free
  swing2LimitDegrees=45
  swing2Motion=Free
  twistLimitDegrees=45
  twistMotion=Free
 }
 {
  bone1Name=dyn_cape_mid_l_05
  bone2Name=dyn_cape_l_05
  swing1LimitDegrees=45
  swing1Motion=Free
  swing2LimitDegrees=45
  swing2Motion=Free
  twistLimitDegrees=45
  twistMotion=Free
 }
 {
  bone1Name=dyn_cape_mid_l_06
  bone2Name=dyn_cape_l_06
  swing1LimitDegrees=45
  swing1Motion=Free
  swing2LimitDegrees=45
  swing2Motion=Free
  twistLimitDegrees=45
  twistMotion=Free
 }
 {
  bone1Name=dyn_cape_mid_l_07
  bone2Name=dyn_cape_l_07
  swing1LimitDegrees=45
  swing1Motion=Free
  swing2LimitDegrees=45
  swing2Motion=Free
  twistLimitDegrees=45
  twistMotion=Free
 }
 {
  bone1Name=dyn_cape_mid_r_01
  bone2Name=dyn_cape_r_01
  swing1LimitDegrees=45
  swing1Motion=Free
  swing2LimitDegrees=45
  swing2Motion=Free
  twistLimitDegrees=45
  twistMotion=Free
 }
 {
  bone1Name=dyn_cape_mid_r_02
  bone2Name=dyn_cape_r_02
  swing1LimitDegrees=45
  swing1Motion=Free
  swing2LimitDegrees=45
  swing2Motion=Free
  twistLimitDegrees=45
  twistMotion=Free
 }
 {
  bone1Name=dyn_cape_mid_r_03
  bone2Name=dyn_cape_r_03
  swing1LimitDegrees=45
  swing1Motion=Free
  swing2LimitDegrees=45
  swing2Motion=Free
  twistLimitDegrees=45
  twistMotion=Free
 }
 {
  bone1Name=dyn_cape_mid_r_04
  bone2Name=dyn_cape_r_04
  swing1LimitDegrees=45
  swing1Motion=Free
  swing2LimitDegrees=45
  swing2Motion=Free
  twistLimitDegrees=45
  twistMotion=Free
 }
 {
  bone1Name=dyn_cape_mid_r_05
  bone2Name=dyn_cape_r_05
  swing1LimitDegrees=45
  swing1Motion=Free
  swing2LimitDegrees=45
  swing2Motion=Free
  twistLimitDegrees=45
  twistMotion=Free
 }
 {
  bone1Name=dyn_cape_mid_r_06
  bone2Name=dyn_cape_r_06
  swing1LimitDegrees=45
  swing1Motion=Free
  swing2LimitDegrees=45
  swing2Motion=Free
  twistLimitDegrees=45
  twistMotion=Free
 }
 {
  bone1Name=dyn_cape_mid_r_07
  bone2Name=dyn_cape_r_07
  swing1LimitDegrees=45
  swing1Motion=Free
  swing2LimitDegrees=45
  swing2Motion=Free
  twistLimitDegrees=45
  twistMotion=Free
 }
]
